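# Unlit debug material in the water fluid simulation folder that renders four numeric values as digits.
# It reads LandscapeLocation and RTWorldSize from the Landmass MPC_Landscape parameter collection and appends their XY components into one float4 for the engine's DebugFloat4Values function.
# Despite the ForceInfo name, it reads no force data, and the graph shows no connection from the function to a material output.
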Material /FortWater/FluidSimulation/Materials/Debug/M_DebugForceInfo
graph {
 properties {pixelDepthOffsetMode=PDOM_Legacy, shadingModel=MSM_Unlit, usedShadingModels=Unlit}
 nodes [
  n1 = CollectionParameter() {collection=/Landmass/Landscape/BlueprintBrushes/MPC/MPC_Landscape, parameterName=LandscapeLocation}
  n2 = CollectionParameter() {collection=&MPC_Landscape, parameterName=RTWorldSize}
  MaterialFunctionCall!(
   Vector4=AppendVector(A=ComponentMask(None=@n1) {g=true, r=true}, B=ComponentMask(None=@n2) {g=true, r=true})
   MaximumNumberOfDigits
   UVs
   DebugTextLocation RG_UpperRight BA_LowerLeft
   Component Spacing
  ) {materialFunction=/Engine/Functions/Engine_MaterialFunctions02/Utility/DebugFloat4Values}
 ]
}
