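# Translucent unlit debug material in the water fluid simulation folder, apparently for displaying forces.
# A tiled landscape alpha texture, bias-scaled and tinted by color parameter C, forms the color term.
# A tunable Fresnel times Opacity and a CustomDepthSort function mask forms the transparency term.
# A separate clamped, rounded-up world-height offset forms a hard height step, with no visible consumer.
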Material /FortWater/FluidSimulation/Materials/Debug/M_ForceDisplay_01
graph {
 properties {blendMode=BLEND_Translucent, pixelDepthOffsetMode=PDOM_Legacy, shadingModel=MSM_Unlit, usedShadingModels=Unlit}
 nodes [
  n1 = ComponentMask(None=WorldPosition().XYZ) {b=true}
  n2 = Constant() {r=64}
  Clamp(None=Ceil(None=Subtract(A=@n1, B=@n2)))
  n3 = TextureCoordinate() {uTiling=8, vTiling=8}
  Multiply(
   A=ConstantBiasScale(None=TextureSample(UVs=@n3) {texture=/Engine/EditorLandscapeResources/DefaultAlphaTexture}.R)
   B=VectorParameter() {parameterName=C, defaultValue=<rgba 0.1 0.5 1 0>}.RGB
  )
  n4 = ScalarParameter() {defaultValue=4, parameterName=Fresnel Exp}
  n5 = ScalarParameter() {defaultValue=0.05, parameterName=Fresnel Base}
  Multiply(
   A=Multiply(A=Fresnel(ExponentIn=@n4, BaseReflectFractionIn=@n5), B=ScalarParameter() {defaultValue=1, parameterName=Opacity})
   B=MaterialFunctionCall() {materialFunction=/FortWater/Materials/Functions/CustomDepthSort}.Mask
  )
 ]
}
