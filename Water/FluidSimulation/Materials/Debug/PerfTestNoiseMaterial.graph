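# Unlit debug material in the fluid simulation folder, named as a performance test for noise cost.
# Its graph holds a single Noise node using the ALU Voronoi function at a small scale.
# No connection from the noise to a material output is shown.
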Material /FortWater/FluidSimulation/Materials/Debug/PerfTestNoiseMaterial
graph {
 properties {bUsedWithNanite=true, pixelDepthOffsetMode=PDOM_Legacy, shadingModel=MSM_Unlit, usedShadingModels=Unlit}
 nodes [Noise!(World Position, FilterWidth) {noiseFunction=NOISEFUNCTION_VoronoiALU, scale=0.05}]
}
